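# Niagara system, filed under FX reuse examples, that draws green energy ribbon trails flowing around a character's leg bones.
# A single burst of persistent particles splits into two spawn groups, each its own ribbon.
# FN_BoneFlow places them along foot, calf and thigh bones from a skinned-mesh sample, using curves for angle and tapering radius.
# Link order drives width and alpha, feeding two ribbon renderers, with a much wider secondary width variable.
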NiagaraSystem /Game/NeverCook/FX/Niagara_FX_Reuse_Examples/SplineThicken/NS_BodyEnergy_Trail_BoneFlow
dependencies {
 usedDataInterfaces [
  &NiagaraDataInterfaceSkeletalMesh
  &NiagaraDataInterfaceArrayFloat
  &NiagaraDataInterfaceCurve
  &NiagaraDataInterfaceColorCurve
 ]
 usedDynamicInputs []
 usedModules [&SystemState]
}
emitters {Trail_Leg {
  emitterUpdateScript [
   &EmitterState(Life Cycle Mode=::NewEnumerator1, Loop Behavior=::NewEnumerator1, Loop Duration Mode=::NewEnumerator1, UseLoopDelay=false)
   &SpawnBurst_Instantaneous(Spawn Count=40, Use Spawn Probability=false, UseLoopCountLimit=false)
  ]
  particleSpawnScript [
   &FN_SetSpawnGroup(Inclusivity=true, Normalize Spawn Group=true, Particles Per Group=20, Pattern=::NewEnumerator0, Set Ribbon ID and Ribbon Link Order=true)
   &InitializeParticle(Lifetime=5, Ribbon Width=1.5, Ribbon Width Mode=::NewEnumerator1, UsePositionOffset=false, Write Material Random=true)
  ]
  particleUpdateScript [
   &ParticleState(Kill Particles When Lifetime Has Elapsed=false, Let Infinitely Lived Particles Die When Emitter Deactivates=true)
   &FN_BoneFlow(
    Angle:NiagaraFloat=&FloatFromCurve(
     FloatCurve=DataInterface() {propertyValues {Curve {keys [
         {arriveTangent=1.8289, interpMode=RCIM_Cubic, leaveTangent=1.8289, tangentMode=RCTM_User, time=0, value=0}
         {arriveTangent=1.7857, interpMode=RCIM_Cubic, leaveTangent=1.7857, time=0.3, value=0.65}
         {arriveTangent=1.5, interpMode=RCIM_Cubic, leaveTangent=1.5, time=0.7, value=1.25}
         {arriveTangent=1.2828, interpMode=RCIM_Cubic, leaveTangent=1.2828, tangentMode=RCTM_User, time=1, value=1.7}
        ]}}}
     CurveIndex=$Particles.Normalized Spawn Group
     Scale Curve=0.8
    )
    BoneID:NiagaraInt32=&ModuloInt(A=$Particles.SpawnGroup, B=2)
    Height:NiagaraFloat=&Multiply_Float(
     A=$Particles.Normalized Spawn Group
     B=&SelectFloatFromArray(Direct Array Index=$Particles.SpawnGroup, Float Selection Array=DataInterface() {propertyValues {FloatData [1.4,1]}})
    )
    Radius:NiagaraFloat=&FloatFromCurve(
     FloatCurve=DataInterface() {propertyValues {Curve {keys [
         {arriveTangent=-0.6799, interpMode=RCIM_Cubic, leaveTangent=-0.6799, tangentMode=RCTM_User, time=0, value=7}
         {arriveTangent=-4.2857, interpMode=RCIM_Cubic, leaveTangent=-4.2857, time=0.3571, value=5.9951}
         {arriveTangent=-8.7317, interpMode=RCIM_Cubic, leaveTangent=-8.7317, tangentMode=RCTM_User, time=0.7, value=4}
         {arriveTangent=-3.5832, interpMode=RCIM_Cubic, leaveTangent=-3.5832, tangentMode=RCTM_User, time=1, value=1}
        ]}}}
     CurveIndex=$Particles.Normalized Spawn Group
     Scale Curve=0.8
    )
    SideVector1:Vector3f=<1 1 1>
    SideVector2:Vector3f=<1 1 1>
    SideVector3:Vector3f=<1 1 1>
    SkeletalMesh=DataInterface() {propertyValues {
      DefaultMesh=null
      ExcludeBoneName=root
      OptionalComponentClass=null
      PreviewMesh=/Game/Characters/Player/Male/Medium/Bodies/M_MED_PiteousKicks/Meshes/M_MED_PiteousKicks
      SkinningMode=SkinOnTheFly
      SoftSourceActor=null
      SourceMode=Default
      UvSetIndex=0
      WholeMeshLOD=-1
      bExcludeBone=false
      bReadDeformedGeometry=true
      bRequireCurrentFrameData=true
      ComponentTags []
      FilteredBones [foot_l,calf_l,thigh_l,foot_r,calf_r,thigh_r]
      FilteredSockets []
      MeshUserParameter {parameter {name=None, typeDefHandle {registeredTypeIndex=1}, varData []}}
      SamplingRegions []
     }}
    Stabalize Twisting=true
   )
   &ScaleRibbonWidth(Ribbon Width Scale=&FloatFromCurve(
     FloatCurve=DataInterface() {propertyValues {Curve {keys [
         <at 0 0.5>
         {arriveTangent=1.8056, interpMode=RCIM_Cubic, leaveTangent=1.8056, tangentMode=RCTM_User, time=0.2, value=0.9}
         {arriveTangent=0.6908, tangentMode=RCTM_User, time=1, value=2}
        ]}}}
     CurveIndex=$Particles.RibbonLinkOrder
    ))
   /SetVariables(Particles.Ribbon B Width:NiagaraFloat=&Multiply_Float(A=$Particles.RibbonWidth, B=8)) {isSetParameters=true}
   &ScaleColor(
    Scale Alpha=&FloatFromCurve(
     FloatCurve=DataInterface() {propertyValues {Curve {keys [
         {arriveTangent=1.4743, interpMode=RCIM_Cubic, leaveTangent=1.4743, tangentMode=RCTM_User, time=0, value=0}
         {arriveTangent=4.159, interpMode=RCIM_Cubic, leaveTangent=-3.8038, tangentMode=RCTM_Break, time=0.3, value=1}
         {arriveTangent=-0.2647, interpMode=RCIM_Cubic, leaveTangent=-0.2647, tangentMode=RCTM_User, time=0.9, value=0}
        ]}}}
     CurveIndex=$Particles.RibbonLinkOrder
     Scale Curve=0.5
    )
    Scale RGB=&MakeVectorFromLinearColorRGB(Color=&ColorFromCurve(
      ColorCurve=DataInterface() {propertyValues {
        AlphaCurve {keys []}
        BlueCurve {keys [<at 0.1537 0.2534 RCIM_Cubic>]}
        GreenCurve {keys [<at 0.1537 1 RCIM_Cubic>]}
        RedCurve {keys [<at 0.1537 0.042 RCIM_Cubic>]}
       }}
      CurveIndex=$StackContext.RibbonLinkOrder
     ))
    ScaleA=true
    ScaleRGB=true
   )
   &ScaleColor(Color Value To Scale=$Particles.Color, ScaleA=true, ScaleRGB=true) {module=ScaleColor001}
   &DynamicMaterialParameters(
    Index 0 Param 3=1
    Index 0 Param 4=1
    Module.Param0WriteEnabled=true
    Module.Param1WriteEnabled=true
    Module.Param2WriteEnabled=true
    Module.Param3WriteEnabled=true
   )
  ]
  rendererClasses [&NiagaraRibbonRendererProperties]
  renderers [
   {rendererClass=&NiagaraRibbonRendererProperties}
   {rendererClass=&NiagaraRibbonRendererProperties, rendererIndex=1}
  ]
 }}
summary {
 systemName=NS_BodyEnergy_Trail_BoneFlow
 emitters [{emitterName=Trail_Leg, rendererClasses [&NiagaraRibbonRendererProperties]}]
 userVariables []
}
