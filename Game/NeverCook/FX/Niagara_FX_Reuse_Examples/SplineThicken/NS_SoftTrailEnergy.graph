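# NS_SoftTrailEnergy is a Niagara system for an energy trail effect, kept among reusable SplineThicken FX examples.
# Its single EnergyTrail emitter burst-spawns a tinted mesh particle with stretched scale, set orientation and a short lifetime.
# A float curve rising over particle life is scaled into a dynamic material parameter, so the mesh material drives the animation.
# It exposes no user variables.
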NiagaraSystem /Game/NeverCook/FX/Niagara_FX_Reuse_Examples/SplineThicken/NS_SoftTrailEnergy
dependencies {usedDataInterfaces [&NiagaraDataInterfaceCurve], usedDynamicInputs [], usedModules [&SystemState]}
emitters {EnergyTrail {
  emitterUpdateScript [
   &EmitterState(Life Cycle Mode=::NewEnumerator1, Loop Behavior=::NewEnumerator1, Loop Delay=0.05, Loop Duration=1, UseLoopDelay=false)
   &SpawnBurst_Instantaneous(Use Spawn Probability=false, UseLoopCountLimit=false)
  ]
  particleSpawnScript [
   &InitializeParticle(
    Color=<rgba 1.5379 1.4069 3 2>
    Lifetime=2
    Mesh Scale=<2 1 1>
    Mesh Scale Mode=::NewEnumerator4
    UsePositionOffset=true
    Write Material Random=true
   )
   &InitialMeshOrientation(Mesh Orientation Mode=::NewEnumerator2, Use Rotation Vector=true)
  ]
  particleUpdateScript [
   &ParticleState()
   &DynamicMaterialParameters(
    Index 0 Param 1=&FloatFromCurve(FloatCurve=DataInterface() {propertyValues {Curve {keys [<at 0 0 RCIM_Cubic>,<at 1 1>]}}}, Scale Curve=5)
    Index 0 Param 3=2
    Module.Index0Float0Write=true
    Module.Index0Float1Write=true
    Module.Index0Float2Write=true
    Module.Index0Float3Write=true
   )
  ]
  rendererClasses [&NiagaraMeshRendererProperties]
  renderers [{rendererClass=&NiagaraMeshRendererProperties, rendererIndex=0}]
 }}
summary {
 systemName=NS_SoftTrailEnergy
 emitters [{emitterName=EnergyTrail, rendererClasses [&NiagaraMeshRendererProperties]}]
 userVariables []
}
