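# Niagara example system for a mesh-based flame effect, named for a spline-thicken technique.
# Its single emitter, Flame01, uses an instantaneous burst to spawn scaled, position-offset mesh particles with material random enabled.
# Particles ignore lifetime expiry and are set to die when the emitter deactivates.
# A dynamic material parameter module feeds values to the mesh material each update, with no user parameters exposed.
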NiagaraSystem /Game/NeverCook/FX/Niagara_FX_Reuse_Examples/SplineThicken/NS_SplineThicken_Flame
dependencies {usedDataInterfaces [], usedDynamicInputs [], usedModules [&SystemState]}
emitters {Flame01 {
  emitterUpdateScript [
   &EmitterState(
    Inactive Response=::NewEnumerator1
    Life Cycle Mode=::NewEnumerator1
    Loop Behavior=::NewEnumerator1
    Loop Duration Mode=::NewEnumerator1
    UseLoopDelay=false
   )
   &SpawnBurst_Instantaneous(Use Spawn Probability=false, UseLoopCountLimit=false)
  ]
  particleSpawnScript [&InitializeParticle(
    Mesh Scale=&Multiply_VectorByFloat(Vector=<1 1 1>, Float=0.9)
    Mesh Scale Mode=::NewEnumerator4
    UsePositionOffset=true
    Write Material Random=true
   )]
  particleUpdateScript [
   &ParticleState(Kill Particles When Lifetime Has Elapsed=false, Let Infinitely Lived Particles Die When Emitter Deactivates=true)
   &DynamicMaterialParameters(
    Index 0 Param 1=3
    Index 0 Param 2=12
    Module.Index0Param0Write=true
    Module.Index0Param1Write=true
    Module.Index0Param2Write=true
    Module.Index0Param3Write=true
   ) {module=DynamicMaterialParameters001}
  ]
  rendererClasses [&NiagaraMeshRendererProperties]
  renderers [{rendererClass=&NiagaraMeshRendererProperties, rendererIndex=0}]
 }}
summary {
 systemName=NS_SplineThicken_Flame
 emitters [{emitterName=Flame01, rendererClasses [&NiagaraMeshRendererProperties]}]
 userVariables []
}
