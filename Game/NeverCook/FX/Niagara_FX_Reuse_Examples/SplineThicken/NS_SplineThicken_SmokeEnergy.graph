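# Niagara mesh effect, named as a SplineThicken smoke-energy example in an FX reuse folder, with a single "smoke" emitter.
# An instantaneous burst spawns Z-stretched mesh particles that are not killed when their lifetime elapses.
# Particle update feeds random ranges into dynamic material parameter indices 0 and 3, varying the material per particle, and applies Scale Color to RGB and alpha.
# No user parameters are exposed and no movement modules are present.
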NiagaraSystem /Game/NeverCook/FX/Niagara_FX_Reuse_Examples/SplineThicken/NS_SplineThicken_SmokeEnergy
dependencies {usedDataInterfaces [], usedDynamicInputs [], usedModules [&SystemState]}
emitters {smoke {
  emitterUpdateScript [
   &EmitterState(
    Inactive Response=::NewEnumerator1
    Life Cycle Mode=::NewEnumerator1
    Loop Behavior=::NewEnumerator1
    Loop Duration Mode=::NewEnumerator1
    UseLoopDelay=false
   )
   &SpawnBurst_Instantaneous(Use Spawn Probability=false, UseLoopCountLimit=false)
  ]
  particleSpawnScript [&InitializeParticle(Lifetime=5, Mesh Scale=<1 1 2>, Mesh Scale Mode=::NewEnumerator4, UsePositionOffset=false, Write Material Random=true)]
  particleUpdateScript [
   &ParticleState(Kill Particles When Lifetime Has Elapsed=false, Let Infinitely Lived Particles Die When Emitter Deactivates=true)
   &DynamicMaterialParameters(
    Index 0 Param 1=&RandomRangeFloat(Minimum=0.5, Maximum=1)
    Index 0 Param 2=&RandomRangeFloat(Minimum=0.8, Maximum=1)
    Index 0 Param 3=&RandomRangeFloat(Minimum=0.5, Maximum=1)
    Index 0 Param 4=&RandomRangeFloat(Minimum=0.5, Maximum=1)
    Index 3 Param 1=&RandomRangeFloat(Minimum=5, Maximum=10)
    Index 3 Param 2=&RandomRangeFloat(Minimum=15, Maximum=25)
    Index 3 Param 3=&RandomRangeFloat(Minimum=3, Maximum=5)
    Index 3 Param 4=&RandomRangeFloat(Minimum=0.5, Maximum=1)
    Module.Index0Float0Write=true
    Module.Index0Float1Write=true
    Module.Index0Float2Write=true
    Module.Index0Float3Write=true
    Module.Index3Float0Write=true
    Module.Index3Float1Write=true
    Module.Index3Float2Write=true
    Module.Index3Float3Write=true
    Write Parameter Index 3=::NewEnumerator1
   )
   &ScaleColor(ScaleA=true, ScaleRGB=true)
  ]
  rendererClasses [&NiagaraMeshRendererProperties]
  renderers [{rendererClass=&NiagaraMeshRendererProperties, rendererIndex=0}]
 }}
summary {
 systemName=NS_SplineThicken_SmokeEnergy
 emitters [{emitterName=smoke, rendererClasses [&NiagaraMeshRendererProperties]}]
 userVariables []
}
